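# Translucent unlit Niagara sprite material for distant streaks in the storm effect.
# Particle color RGB feeds the MF_FX_EmissiveAdjustments function for glow.
# Two vertically panning line-mask layers at different tilings and speeds are summed, each shifted by a dynamic parameter that can offset it per particle.
# A third panning streak texture is inverted to break up that sum, which is then scaled by particle alpha.
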Material /FNGameplayCues/Effects/Fort_Effects/Effects/Environment/Storm/Materials/M_Storm_Distant_Streaks
graph {
 properties {bUsedWithNiagaraSprites=true, blendMode=BLEND_Translucent, pixelDepthOffsetMode=PDOM_Legacy, shadingModel=MSM_Unlit, usedShadingModels=Unlit}
 nodes [
  n1 = ParticleColor()
  MaterialFunctionCall(Emissive=@n1.RGB) {materialFunction=&MF_FX_EmissiveAdjustments}
  n2 = DynamicParameter() {paramNames [X,Y,X2,Y2]}
  n3 = TextureCoordinate() {uTiling=0.2, vTiling=5}
  n4 = TextureSample(UVs=Add(A=AppendVector(A=@n2.X, B=@n2.X), B=Panner(Coordinate=@n3) {speedY=1})) {samplerType=SAMPLERTYPE_Masks, texture=/FNGameplayCues/Athena/Effects/Particles/Skydiving/NS_Contrail_Starfield/T_Junior_ZeroLines_M}
  n5 = DynamicParameter() {paramNames [X,Y,X2,Y2]}
  n6 = TextureCoordinate() {vTiling=6}
  n7 = TextureSample(UVs=Add(A=AppendVector(A=@n5.X2, B=@n5.X2), B=Panner(Coordinate=@n6) {speedY=2})) {samplerType=SAMPLERTYPE_Masks, texture=&T_Junior_ZeroLines_M}
  n8 = Panner(Coordinate=TextureCoordinate() {vTiling=10}) {speedY=1}
  n9 = Constant2Vector() {g=2, r=4}
  n10 = Multiply(A=TextureSample(UVs=Multiply(A=@n8, B=@n9)) {samplerType=SAMPLERTYPE_Masks, texture=/Game/Environments/Sets/Vulcano/Textures/T_LineStreaks_M}.R)
  Multiply(A=@n1.A, B=Multiply(A=Add(A=@n4.G, B=@n7.G), B=OneMinus(None=@n10)))
 ]
}
